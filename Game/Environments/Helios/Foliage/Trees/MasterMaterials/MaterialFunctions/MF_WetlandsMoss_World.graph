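# Material function that layers world-projected jungle moss over incoming material attributes, for Helios tree materials.
# Coverage comes from height-lerping moss height against either a vertex-color channel or three skewed, multiplied large-scale world-space samples.
# Moss combines world-aligned diffuse with hue-shifted Fresnel tint, height-driven specular and roughness, and angle-corrected normals, plus a simpler low-end variant.
# The blend is behind UseJungleMoss and Epic quality; mobile and other quality levels pass input through.
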MaterialFunction /Game/Environments/Helios/Foliage/Trees/MasterMaterials/MaterialFunctions/MF_WetlandsMoss_World
graph {nodes [
  n1 = ScalarParameter() {defaultValue=0.5, group=JungleMoss, parameterName=JungleMoss_C3}
  n2 = ScalarParameter() {defaultValue=20, group=JungleMoss, parameterName=JungleMoss_C4}
  n3 = ComponentMask(None=MaterialFunctionCall(TextureObject=NamedRerouteUsage(), TextureSize=NamedRerouteUsage()) {materialFunction=&WorldAlignedTexture}.XYZ Texture) {r=true}
  n4 = LinearInterpolate(
   A=ScalarParameter() {group=JungleMoss, parameterName=JungleMoss_VertexColor_C1}
   B=ScalarParameter() {defaultValue=1, group=JungleMoss, parameterName=JungleMoss_VertexColor_C2}
   Alpha=ChannelMaskParameter(None=VertexColor().A) {group=JungleMoss, parameterName=VertexColor_Moss}.R
  )
  n5 = WorldPosition()
  n6 = VectorParameter() {group=JungleMoss, parameterName=JungleTexture_Large_Offset, defaultValue=<rgba 278.1917 309.5966 0 1>}
  n7 = MaterialFunctionCall(Float3=Divide(A=Add(A=@n5.XYZ, B=@n6.RGB), B=ScalarParameter() {defaultValue=3300, group=JungleMoss, parameterName=JungleMoss_Large_Size_Mask})) {materialFunction=&BreakOutFloat3Components}
  n8 = ScalarParameter() {defaultValue=0.01, group=JungleMoss, parameterName=JungleMoss_Large_Skew}
  n9 = Add(A=@n7.R, B=Multiply(A=@n7.G, B=@n8))
  n10 = Add(A=@n7.B, B=Multiply(A=Add(A=@n7.R, B=@n7.R), B=@n8))
  n11 = TextureSample(UVs=AppendVector(A=@n9, B=@n10), Tex=NamedRerouteUsage()) {texture=/Game/Environments/Asteria/Terrain/Textures/T_Asteria_JungleMoss_H}
  n12 = Add(A=@n7.G, B=Multiply(A=@n7.R, B=@n8))
  n13 = TextureSample(UVs=AppendVector(A=@n12, B=@n10), Tex=NamedRerouteUsage()) {texture=&T_Asteria_JungleMoss_H}
  n14 = AppendVector(A=@n9, B=@n12)
  n15 = NamedRerouteUsage()
  n16 = LinearInterpolate(
   A=ScalarParameter() {defaultValue=0.2, group=JungleMoss, parameterName=JungleMoss_Large_C1}
   B=ScalarParameter() {defaultValue=800, group=JungleMoss, parameterName=JungleMoss_Large_C2}
   Alpha=Multiply(
    A=Multiply(A=@n11.R, B=@n13.R)
    B=TextureSample(UVs=@n14, Tex=@n15) {texture=&T_Asteria_JungleMoss_H}.R
   )
  )
  n17 = MaterialFunctionCall(
   Transition Phase=StaticSwitchParameter(True=Saturate(None=@n4), False=Saturate(None=@n16)) {group=JungleMoss, parameterName=UseVertexColorJungleMoss}
   Height Texture=@n3
   Contrast=ScalarParameter() {defaultValue=0.5, group=JungleMoss, parameterName=JungleMoss_Contrast}
  ) {materialFunction=&HeightLerp}
  n18 = Multiply(A=Saturate(None=LinearInterpolate(A=@n1, B=@n2, Alpha=@n3)), B=@n17.Alpha)
  Add(A=NamedRerouteUsage(), B=@n18)
  n19 = ScalarParameter() {defaultValue=600, group=JungleMoss, parameterName=JungleMoss_Size}
  n20 = NamedRerouteDeclaration(None=@n19) {name=MossTextureScale, nodeColor=<rgba 0 1 0.4844 1>}
  n21 = ScalarParameter() {defaultValue=1, group=JungleMoss, parameterName=JungleMoss_PrjectionContrast}
  n22 = MaterialFunctionCall(TextureObject=NamedRerouteUsage(), TextureSize=@n20, ProjectionTransitionContrast=@n21) {materialFunction=&WorldAlignedTexture}
  ComponentMask(None=@n22.XYZ Texture) {g=true}
  n23 = GetMaterialAttributes(None=FunctionInput() {inputType=FunctionInput_MaterialAttributes}) {attributeGetTypes [#1, #2]}
  n24 = NamedRerouteUsage()
  n25 = NamedRerouteUsage()
  n26 = QualitySwitch(Default=@n18, Low=@n17.Alpha, High=@n17.Alpha, Medium=@n17.Alpha)
  n27 = StaticSwitchParameter(
   True=BlendMaterialAttributes(A=@n23.MaterialAttributes, B=QualitySwitch(Default=@n24, Low=@n25, High=@n25, Medium=@n25, Epic=@n24), Alpha=Saturate(None=@n26)) {vertexAttributeBlendType=UseA}
   False=@n23.MaterialAttributes
  ) {group=JungleMoss, parameterName=UseJungleMoss}
  FunctionOutput(None=ShadingPathSwitch(Default=QualitySwitch(Default=@n23.MaterialAttributes, Epic=@n27), Mobile=@n23.MaterialAttributes) {desc=Migrated from FeatureLevelSwitch})
  LinearInterpolate(
   A=ScalarParameter() {defaultValue=0.2, group=JungleMoss, parameterName=JungleMoss_C1}
   B=ScalarParameter() {defaultValue=6.61, group=JungleMoss, parameterName=JungleMoss_C2}
   Alpha=MaterialFunctionCall(TextureObject=NamedRerouteUsage(), TextureSize=ScalarParameter() {defaultValue=5300, group=JungleMoss, parameterName=JungleMoss_SizeMask}) {materialFunction=&WorldAlignedTexture}.XYZ Texture
  )
  n28 = ScalarParameter() {defaultValue=0.3, group=JungleMoss, parameterName=Mobile_Color_Mask_Low}
  n29 = ScalarParameter() {defaultValue=3.209, group=JungleMoss, parameterName=Mobile_Color_Mask_High}
  LinearInterpolate(
   A=VectorParameter() {group=JungleMoss, parameterName=Mobile_Moss_Color_1, defaultValue=<rgba 0.0272 0.055 0.0136 1>}.RGB
   B=VectorParameter() {group=JungleMoss, parameterName=Mobile_Moss_Color_2, defaultValue=<rgba 0.0379 0.07 0.0106 1>}.RGB
   Alpha=Saturate(None=LinearInterpolate(A=@n28, B=@n29, Alpha=@n3))
  )
  Multiply()
  n30 = Fresnel(ExponentIn=ScalarParameter() {defaultValue=0.8607, group=JungleMoss, parameterName=JungleMoss_Fresnell}, Normal=VertexNormalWS()) {baseReflectFraction=0}
  n31 = VectorParameter() {group=JungleMoss, parameterName=JungleMoss_FresnelColor, defaultValue=<rgba 0.5157 1 0.439 1>}
  n32 = ScalarParameter() {defaultValue=-0.172, group=JungleMoss, parameterName=JungleMoss_Hue}
  n33 = Add(
   A=Multiply(
    A=Multiply(A=@n30, B=@n31.RGB)
    B=MaterialFunctionCall(Hue Shift Percentage=@n32, Texture=@n22.XYZ Texture) {materialFunction=&HueShift}.Result
   )
   B=Desaturation(None=@n22.XYZ Texture, Fraction=ScalarParameter() {defaultValue=0.2, group=JungleMoss, parameterName=JungleMoss_Desat})
  )
  n34 = LinearInterpolate(
   A=ScalarParameter() {defaultValue=0.112, group=JungleMoss, parameterName=JungleMoss_Specular_low}
   B=ScalarParameter() {defaultValue=0.5, group=JungleMoss, parameterName=JungleMoss_Specular_high}
   Alpha=@n3
  )
  n35 = LinearInterpolate(
   A=ScalarParameter() {defaultValue=0.7, group=JungleMoss, parameterName=JungleMoss_Roughness_low}
   B=ScalarParameter() {defaultValue=1, group=JungleMoss, parameterName=JungleMoss_Roughness_high}
   Alpha=@n3
  )
  n36 = MaterialFunctionCall(TextureObject=NamedRerouteUsage(), TextureSize=@n20, ProjectionTransitionContrast=@n21) {materialFunction=&WorldAlignedNormal}
  n37 = ScalarParameter() {defaultValue=0.5, group=JungleMoss, parameterName=JungleMoss_Normal}
  n38 = SetMaterialAttributes!(
   MaterialAttributes
   BaseColor=@n33
   Specular=@n34
   Roughness=@n35
   Normal=MaterialFunctionCall(
    BaseNormal=NamedRerouteUsage()
    AdditionalNormal=MaterialFunctionCall(Normal=@n36.XYZ Texture, Flatness=@n37) {materialFunction=&FlattenNormal}.Result
   ) {materialFunction=&BlendAngleCorrectedNormals}.Result
  ) {attributeSetTypes [
    #3
    #4
    #2
    #1
   ]}
  NamedRerouteDeclaration(None=@n38) {name=MossHighEnd, nodeColor=<rgba 1 0.5156 0 1>}
  NamedRerouteDeclaration(None=@n23.Normal) {name=OriginalNormal, nodeColor=<rgba 0 1 0.7188 1>}
  n39 = SetMaterialAttributes(BaseColor=@n33, Specular=Constant() {r=0.5}, Roughness=Constant() {r=1}, Normal=NamedRerouteUsage()) {attributeSetTypes [
    #3
    #4
    #2
    #1
   ]}
  NamedRerouteDeclaration(None=@n39) {name=LowEndMossMaterial, nodeColor=<rgba 0 0.9297 1 1>}
  n40 = ScalarParameter() {defaultValue=0.7, group=JungleMoss, parameterName=Detail_Moss_Mask_TransitionAmount}
  n41 = LinearInterpolate(
   A=ScalarParameter() {defaultValue=-1.7139, group=JungleMoss, parameterName=Detail_Moss_Mask Min}
   B=ScalarParameter() {defaultValue=2, group=JungleMoss, parameterName=Detail_Moss_Mask Max}
   Alpha=@n23.Roughness
  )
  n42 = ScalarParameter() {group=JungleMoss, parameterName=Detail_Moss_Mask_Contrast}
  NamedRerouteDeclaration(None=OneMinus(None=MaterialFunctionCall(Transition Phase=@n40, Height Texture=@n41, Contrast=@n42) {materialFunction=&HeightLerp}.Alpha)) {name=RoughnessBaseMask, nodeColor=<rgba 0 1 0.5078 1>}
  n43 = TextureObjectParameter() {group=JungleMoss, parameterName=Moss Height Texture, texture=&T_Asteria_JungleMoss_H}
  NamedRerouteDeclaration(None=@n43) {name=T_Height, nodeColor=<rgba 0.9219 1 0 1>}
  n44 = TextureObjectParameter() {
   group=JungleMoss
   parameterName=Moss Normal Texture
   samplerType=SAMPLERTYPE_Normal
   texture=/Game/Environments/Asteria/Terrain/Textures/T_Asteria_JungleMoss_N
  }
  NamedRerouteDeclaration(None=@n44) {name=T_Normal, nodeColor=<rgba 0 1 0.4375 1>}
  n45 = TextureObjectParameter() {group=JungleMoss, parameterName=Moss Diffuse Texture, texture=/Game/Environments/Asteria/Terrain/Textures/T_Asteria_JungleMoss_D}
  NamedRerouteDeclaration(None=@n45) {name=T_Diffuse, nodeColor=<rgba 1 0.1172 0 1>}
  Reroute(None=@n38)
  Reroute(None=@n23.Normal)
  Reroute(None=@n19)
  Reroute(None=@n39)
  Reroute(None=@n34)
  Reroute(None=@n35)
 ]}
